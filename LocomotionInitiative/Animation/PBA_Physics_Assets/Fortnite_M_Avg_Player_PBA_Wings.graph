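# Physics asset for the average male Fortnite player skeleton, named as a physics-based animation variant for wings, though it holds no wing or arm bodies.
# Capsules cover pelvis, five spine bones and head, with two capsules per thigh and calf and a box per foot.
# Constraints chain the spine, head and legs to the pelvis.
# All set angular motion Free, leaving their stored limits inactive, except the knees, which keep default motion with hinge-like limits.
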PhysicsAsset /LocomotionInitiative/Animation/PBA_Physics_Assets/Fortnite_M_Avg_Player_PBA_Wings
bodies {
 calf_l [
  {shapeType=Capsule, length=18.169, radius=6.1872, shapeName=calf_l_capsule, center=<-31.0315 -1.0114 0>, rotation=<rot 90 180 -175>}
  {shapeType=Capsule, length=10.7677, radius=6.7999, shapeName=calf_l_capsule, center=<-11.769 -2.5363 0.4807>, rotation=<rot 90 0 0>}
 ]
 calf_r [
  {shapeType=Capsule, length=18.169, radius=6.1872, shapeName=calf_l_capsule, center=<31.0315 1.0114 0>, rotation=<rot 90 180 5>}
  {shapeType=Capsule, length=10.7677, radius=6.7999, shapeName=calf_l_capsule, center=<11.769 2.5363 -0.4807>, rotation=<rot 90 0 180>}
 ]
 foot_l [{shapeType=Box, extentX=13.717, extentY=23.2938, extentZ=13.2787, shapeName=foot_l_box, center=<-4.7256 3.3842 -0.2529>}]
 foot_r [{
   shapeType=Box
   extentX=13.717
   extentY=23.2938
   extentZ=13.2787
   shapeName=foot_l_box
   center=<4.7256 -3.3842 0.2529>
   rotation=<rot 0 0 -180>
  }]
 head [{shapeType=Capsule, length=3.1392, radius=6.3177, shapeName=head_capsule, center=<7.3296 1.9947 0>, rotation=<rot 0 -0.0002 -90>}]
 pelvis [{shapeType=Capsule, length=9.3633, radius=11.7835, shapeName=pelvis_capsule, center=<-2.9195 -1.3575 0>, rotation=<rot 0 -0.0002 0>}]
 spine_01 [{shapeType=Capsule, length=7.6877, radius=10.2524, shapeName=spine_01_capsule, center=<2.4166 0.492 0>, rotation=<rot 0 -0.0002 0>}]
 spine_02 [{shapeType=Capsule, length=6.8371, radius=10.0925, shapeName=spine_02_capsule, center=<0.6394 1.089 0>, rotation=<rot 0 -0.0002 0>}]
 spine_03 [{shapeType=Capsule, length=6.6262, radius=11.0214, shapeName=spine_03_capsule, center=<2.4064 1.3054 0>, rotation=<rot 0 -0.0002 0>}]
 spine_04 [{shapeType=Capsule, length=5.4325, radius=12.6263, shapeName=spine_04_capsule, center=<2.2037 0.3395 0>, rotation=<rot 0 -0.0002 0>}]
 spine_05 [{shapeType=Capsule, length=12.4582, radius=11.2091, shapeName=spine_05_capsule, center=<-0.179 -0.5192 0>, rotation=<rot 0 -0.0002 0>}]
 thigh_l [
  {shapeType=Capsule, length=13.3387, radius=6.6923, shapeName=thigh_l_capsule, center=<-30.0793 0 0>, rotation=<rot 90 0 0>}
  {
   shapeType=Capsule
   length=12.2552
   radius=7.7758
   shapeName=thigh_l_capsule
   center=<-12.0276 0.6183 -0.1645>
   rotation=<rot 88.0333 0.0006 0.0012>
  }
 ]
 thigh_r [
  {shapeType=Capsule, length=13.3387, radius=6.6923, shapeName=thigh_l_capsule, center=<30.0793 0 0>, rotation=<rot 90 0 180>}
  {
   shapeType=Capsule
   length=12.2552
   radius=7.7758
   shapeName=thigh_l_capsule
   center=<12.0276 -0.6183 0.1645>
   rotation=<rot 88.0333 0.0006 -179.9988>
  }
 ]
}
constraints [
 {
  bone1Name=spine_01
  bone2Name=pelvis
  swing1LimitDegrees=45
  swing1Motion=Free
  swing2LimitDegrees=45
  swing2Motion=Free
  twistLimitDegrees=45
  twistMotion=Free
 }
 {
  bone1Name=spine_02
  bone2Name=spine_01
  swing1LimitDegrees=45
  swing1Motion=Free
  swing2LimitDegrees=45
  swing2Motion=Free
  twistLimitDegrees=45
  twistMotion=Free
 }
 {
  bone1Name=spine_03
  bone2Name=spine_02
  swing1LimitDegrees=45
  swing1Motion=Free
  swing2LimitDegrees=45
  swing2Motion=Free
  twistLimitDegrees=45
  twistMotion=Free
 }
 {
  bone1Name=spine_04
  bone2Name=spine_03
  swing1LimitDegrees=45
  swing1Motion=Free
  swing2LimitDegrees=45
  swing2Motion=Free
  twistLimitDegrees=45
  twistMotion=Free
 }
 {
  bone1Name=spine_05
  bone2Name=spine_04
  swing1LimitDegrees=45
  swing1Motion=Free
  swing2LimitDegrees=45
  swing2Motion=Free
  twistLimitDegrees=45
  twistMotion=Free
 }
 {
  bone1Name=foot_l
  bone2Name=calf_l
  swing1LimitDegrees=35
  swing1Motion=Free
  swing2LimitDegrees=15
  swing2Motion=Free
  twistLimitDegrees=10
  twistMotion=Free
 }
 {bone1Name=calf_l, bone2Name=thigh_l, swing1LimitDegrees=70, swing2LimitDegrees=1, twistLimitDegrees=1}
 {
  bone1Name=thigh_l
  bone2Name=pelvis
  swing1LimitDegrees=70
  swing1Motion=Free
  swing2LimitDegrees=45
  swing2Motion=Free
  twistLimitDegrees=15
  twistMotion=Free
 }
 {bone1Name=calf_r, bone2Name=thigh_r, swing1LimitDegrees=70, swing2LimitDegrees=1, twistLimitDegrees=1}
 {
  bone1Name=foot_r
  bone2Name=calf_r
  swing1LimitDegrees=35
  swing1Motion=Free
  swing2LimitDegrees=15
  swing2Motion=Free
  twistLimitDegrees=10
  twistMotion=Free
 }
 {
  bone1Name=head
  bone2Name=spine_05
  swing1LimitDegrees=45
  swing1Motion=Free
  swing2LimitDegrees=30
  swing2Motion=Free
  twistLimitDegrees=45
  twistMotion=Free
 }
 {
  bone1Name=thigh_r
  bone2Name=pelvis
  swing1LimitDegrees=70
  swing1Motion=Free
  swing2LimitDegrees=45
  swing2Motion=Free
  twistLimitDegrees=15
  twistMotion=Free
 }
]
