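# Player linked anim graph adding physics-based animation for DBNO, skydiving, wings, hover pack, PuppetBoss and IcicleLarva states.
# RigidBodyWithControl nodes simulate the cached input pose; nested bool blends pick one, DBNO first, or pass input through.
# BeginPlay registers Character Dynamics state-logic layers, which node update functions use to set control parameters.
# DBNO requires a curve-table toggle; the DBNO-carried branch is not wired to output.
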AnimBlueprint /LocomotionInitiative/Animation/AnimationBlueprints/LinkedAnimGraphs/Player_RigidBodyPBA_AnimBP
target_skeleton=&Fortnite_M_Avg_Player_Skeleton
graphs [
 AnimGraph {nodes [
   AnimGraphNode_RigidBodyWithControl_6 = RigidBodyWithControl!(
    ComponentPose:Component Space Pose Link Structure=Local To Component(LocalPose=Use cached pose 'startCache'())
    Alpha:float=1
    ExternalForce:Vector="0,0,0"
    ForcesAndTorques:Rigid Body Forces And Torques Structure=(ForcesAndTorques=())
   ) -> (Pose:Component Space Pose Link Structure) {
    is=title AnimGraphNode_RigidBodyWithControl
    binding=&self:AnimGraph.AnimGraphNode_RigidBodyWithControl_6.Binding
    node {
     bEnableControls=true
     bOverrideWorldGravity=true
     evaluationResetTime=0.2
     overridePhysicsAsset=/LocomotionInitiative/Animation/PBA_Physics_Assets/Fortnite_M_Avg_Player_PBA_DBNO_Carried
     physicsAssetAuthoredSkeletalMesh=/Game/Characters/Player/Male/Medium/Base/SK_M_MALE_Base_Skeleton
     physicsAssetConditioningConstraintOrientation=AuthoredSkeleton
     physicsAssetConditioningConstraintPosition=AuthoredSkeleton
     additionalSets {
      controlSetUpdates [
       {setName=WorldSpace_Arms, names [WorldSpace_ArmL]}
       {setName=ParentSpace_Arms, names [ParentSpace_ArmL]}
       {
        setName=WorldSpace_Fingers
        names [index_01_l,index_02_l,middle_01_l,middle_02_l,ring_01_l,ring_02_l,pinky_01_l,pinky_02_l,thumb_01_l,thumb_02_l]
       }
       {setName=ParentSpace_Legs, names [ParentSpace_LegL, ParentSpace_LegR]}
       {setName=WorldSpace_Legs, names [WorldSpace_LegL, WorldSpace_LegR]}
       {setName=WorldSpace_Hands, names [hand_l]}
       {setName=WorldSpace_Feet, names [foot_l,foot_r]}
      ]
      modifierSetUpdates [{setName=hands, names [hand_l]}, {setName=feet, names [foot_l,foot_r]}, {setName=Arms, names [ArmL]}]
     }
     alphaBoolBlend {blendOption=Linear}
     characterSetupData {
      defaultParentSpaceControlData {bEnabled=false}
      defaultWorldSpaceControlData {bEnabled=false}
      limbSetupData [
       {
        bCreateBodyModifiers=true
        bCreateParentSpaceControls=true
        bCreateWorldSpaceControls=true
        bIncludeParentBone=false
        limbName=Head
        startBone=neck_01
       }
       {
        bCreateBodyModifiers=true
        bCreateParentSpaceControls=true
        bCreateWorldSpaceControls=true
        bIncludeParentBone=false
        limbName=ArmL
        startBone=clavicle_l
       }
       {
        bCreateBodyModifiers=true
        bCreateParentSpaceControls=true
        bCreateWorldSpaceControls=true
        bIncludeParentBone=false
        limbName=LegL
        startBone=thigh_l
       }
       {
        bCreateBodyModifiers=true
        bCreateParentSpaceControls=true
        bCreateWorldSpaceControls=true
        bIncludeParentBone=false
        limbName=LegR
        startBone=thigh_r
       }
       {
        bCreateBodyModifiers=true
        bCreateParentSpaceControls=true
        bCreateWorldSpaceControls=true
        bIncludeParentBone=true
        limbName=Spine
        startBone=spine_05
       }
      ]
     }
     overrideWorldGravity {z=-980}
     updateFunction {functionName=OnUpdate}
    }
    updateFunction {bSelfContext=true, memberGuid=#1, memberName=On Update_Skydiving}
   }
   Blend Poses by bool(BlendPose_0=Component To Local(ComponentPose=@AnimGraphNode_RigidBodyWithControl_6), bActiveValue=Get bIsDBNOCarried()) {blendType=Linear}
   n1 = "Sequence Player (additive) 'Dive_Additive'"() {sequence=/Game/Animation/Game/MainPlayer/Skydive/Freefall/Medium/Female/Dive_Additive}
   n2 = Use cached pose 'SkydivePBA'()
   n3 = Blend Poses by bool(
    BlendPose_0=Blend Poses by bool(
     BlendPose_0=Apply Additive(Base=@n2, Additive=@n1) {alphaBoolBlend {blendOption=Linear}}
     BlendPose_1=Use cached pose 'SkydivePBA'()
     bActiveValue=Get bIsSkydiveDiveMode()
    ) {blendType=Linear}
    BlendPose_1=Use cached pose 'startCache'()
    bActiveValue=Get bUsePBAInAnimGraph() -> (bUsePBAInAnimGraph:bool)
   ) {blendType=Linear}
   n4 = Use cached pose 'WingsPBA'()
   n5 = Get bIsWingFlying() -> (bIsWingFlying:bool)
   n6 = Blend Poses by bool(
    BlendPose_0=Use cached pose 'PuppetPBA'()
    BlendPose_1=Blend Poses by bool(
     BlendPose_0=Use cached pose 'HoverPBA'()
     BlendPose_1=Blend Poses by bool(BlendPose_0=@n4, BlendPose_1=@n3, bActiveValue=@n5) {blendType=Linear}
     bActiveValue=Get bIsHovering() -> (bIsHovering:bool)
    ) {blendType=Linear}
    bActiveValue=Get bPuppetBoss() -> (bPuppetBoss:bool)
   ) {blendType=Linear}
   n7 = Use cached pose 'IcicleLarva'()
   n8 = Get IcicleLarva_MovementState() -> (IcicleLarva_MovementState:bool)
   Output Pose(Result=Blend Poses by bool(
     BlendPose_0=Use cached pose 'DBNOActionsPBA'()
     BlendPose_1=Blend Poses by bool(BlendPose_0=@n7, BlendPose_1=@n6, bActiveValue=@n8) {blendType=Linear}
     bActiveValue=Get PBA_DBNOActive() -> (PBA_DBNOActive:bool)
    ) {blendType=Linear}) {layerGroup=None}
   Save cached pose 'startCache'(Pose=Input Pose()) {cacheName=startCache}
   AnimGraphNode_RigidBodyWithControl_2 = RigidBodyWithControl!(
    ComponentPose=Local To Component(LocalPose=Use cached pose 'startCache'())
    AlphaCurveName:Name=pbadisabled
    ExternalForce="0,0,0"
    ForcesAndTorques=(ForcesAndTorques=())
   ) -> (Pose) {
    binding=&self:AnimGraph.AnimGraphNode_RigidBodyWithControl_2.Binding
    node {
     alphaInputType=Curve
     bEnableControls=true
     bOverrideWorldGravity=true
     evaluationResetTime=0.2
     overridePhysicsAsset=/LocomotionInitiative/Animation/PBA_Physics_Assets/Fortnite_M_Avg_Player_PBA_Skydive
     physicsAssetAuthoredSkeletalMesh=&SK_M_MALE_Base_Skeleton
     physicsAssetConditioningConstraintOrientation=AuthoredSkeleton
     physicsAssetConditioningConstraintPosition=AuthoredSkeleton
     additionalSets {
      controlSetUpdates [
       {setName=WorldSpace_Arms, names [WorldSpace_ArmL, WorldSpace_ArmR]}
       {setName=ParentSpace_Arms, names [ParentSpace_ArmL, ParentSpace_ArmR]}
       {setName=ParentSpace_Legs, names [ParentSpace_LegL, ParentSpace_LegR]}
       {setName=WorldSpace_Legs, names [WorldSpace_LegL, WorldSpace_LegR]}
       {setName=WorldSpace_Hands, names [hand_l,hand_r]}
       {setName=WorldSpace_Feet, names [foot_l,foot_r]}
       {setName=WorldSpace_LowerSpine, names [spine_04,spine_03,spine_02,spine_01,pelvis]}
      ]
      modifierSetUpdates [{setName=hands, names [hand_l,hand_r,weapon_r]}, {setName=feet, names [foot_l,foot_r]}, {setName=Arms, names [ArmL,ArmR]}]
     }
     alphaBoolBlend {blendOption=Linear}
     alphaScaleBiasClamp {bias=1, scale=-1}
     characterSetupData {
      defaultParentSpaceControlData {bEnabled=false}
      defaultWorldSpaceControlData {bEnabled=false}
      limbSetupData [
       {
        bCreateBodyModifiers=true
        bCreateParentSpaceControls=true
        bCreateWorldSpaceControls=true
        bIncludeParentBone=false
        limbName=Head
        startBone=neck_01
       }
       {
        bCreateBodyModifiers=true
        bCreateParentSpaceControls=true
        bCreateWorldSpaceControls=true
        bIncludeParentBone=false
        limbName=ArmL
        startBone=clavicle_l
       }
       {
        bCreateBodyModifiers=true
        bCreateParentSpaceControls=true
        bCreateWorldSpaceControls=true
        bIncludeParentBone=false
        limbName=ArmR
        startBone=clavicle_r
       }
       {
        bCreateBodyModifiers=true
        bCreateParentSpaceControls=true
        bCreateWorldSpaceControls=true
        bIncludeParentBone=false
        limbName=LegL
        startBone=thigh_l
       }
       {
        bCreateBodyModifiers=true
        bCreateParentSpaceControls=true
        bCreateWorldSpaceControls=true
        bIncludeParentBone=false
        limbName=LegR
        startBone=thigh_r
       }
       {
        bCreateBodyModifiers=true
        bCreateParentSpaceControls=true
        bCreateWorldSpaceControls=true
        bIncludeParentBone=true
        limbName=Spine
        startBone=spine_01
       }
      ]
     }
     updateFunction {functionName=OnUpdate}
    }
    updateFunction {bSelfContext=true, memberGuid=#1, memberName=On Update_Skydiving}
   }
   Save cached pose 'SkydivePBA'(Pose=Component To Local(ComponentPose=@AnimGraphNode_RigidBodyWithControl_2)) {cacheName=SkydivePBA}
   AnimGraphNode_RigidBodyWithControl_4 = RigidBodyWithControl!(
    ComponentPose=Local To Component(LocalPose=Use cached pose 'startCache'())
    Alpha:float=1
    ExternalForce="0,0,0"
    ForcesAndTorques=(ForcesAndTorques=())
   ) -> (Pose) {
    binding=&self:AnimGraph.AnimGraphNode_RigidBodyWithControl_4.Binding
    node {
     bEnableControls=true
     bOverrideWorldGravity=true
     evaluationResetTime=0.2
     overridePhysicsAsset=/LocomotionInitiative/Animation/PBA_Physics_Assets/Fortnite_M_Avg_Player_PBA_Wings
     physicsAssetAuthoredSkeletalMesh=&SK_M_MALE_Base_Skeleton
     physicsAssetConditioningConstraintOrientation=AuthoredSkeleton
     physicsAssetConditioningConstraintPosition=AuthoredSkeleton
     additionalSets {
      controlSetUpdates [
       {setName=ParentSpace_Legs, names [ParentSpace_LegL, ParentSpace_LegR]}
       {setName=WorldSpace_Legs, names [WorldSpace_LegL, WorldSpace_LegR]}
       {setName=WorldSpace_Feet, names [foot_l,foot_r]}
       {setName=WorldSpace_LowerSpine, names [spine_04,spine_03,spine_02,spine_01,pelvis]}
      ]
      modifierSetUpdates [{setName=Legs, names [LegL,LegR]}]
     }
     alphaBoolBlend {blendOption=Linear}
     characterSetupData {
      defaultParentSpaceControlData {bEnabled=false}
      defaultWorldSpaceControlData {bEnabled=false}
      limbSetupData [
       {
        bCreateBodyModifiers=true
        bCreateParentSpaceControls=true
        bCreateWorldSpaceControls=true
        bIncludeParentBone=false
        limbName=LegL
        startBone=thigh_l
       }
       {
        bCreateBodyModifiers=true
        bCreateParentSpaceControls=true
        bCreateWorldSpaceControls=true
        bIncludeParentBone=false
        limbName=LegR
        startBone=thigh_r
       }
       {
        bCreateBodyModifiers=true
        bCreateParentSpaceControls=true
        bCreateWorldSpaceControls=true
        bIncludeParentBone=true
        limbName=Spine
        startBone=spine_01
       }
      ]
     }
     updateFunction {functionName=OnUpdate_Wings}
    }
    updateFunction {bSelfContext=true, memberGuid=#2, memberName=OnUpdate_Wings}
   }
   Save cached pose 'WingsPBA'(Pose=Component To Local(ComponentPose=@AnimGraphNode_RigidBodyWithControl_4)) {cacheName=WingsPBA}
   AnimGraphNode_RigidBodyWithControl_0 = RigidBodyWithControl!(
    ComponentPose=Local To Component(LocalPose=Use cached pose 'startCache'())
    AlphaCurveName:Name=pbadisabled
    ExternalForce="0,0,0"
    ForcesAndTorques=(ForcesAndTorques=())
   ) -> (Pose) {
    binding=&self:AnimGraph.AnimGraphNode_RigidBodyWithControl_0.Binding
    node {
     alphaInputType=Curve
     bEnableControls=true
     bEnableWorldGeometry=true
     bOverrideWorldGravity=true
     bWorldGeometryCollisionEnabledHasQuery=true
     evaluationResetTime=0.2
     overridePhysicsAsset=/LocomotionInitiative/Animation/PBA_Physics_Assets/Fortnite_M_Avg_Player_PBA_DBNO_Tumbling
     physicsAssetAuthoredSkeletalMesh=&SK_M_MALE_Base_Skeleton
     physicsAssetConditioningConstraintOrientation=AuthoredSkeleton
     physicsAssetConditioningConstraintPosition=AuthoredSkeleton
     additionalSets {
      controlSetUpdates [
       {setName=WorldSpace_Arms, names [WorldSpace_ArmL, WorldSpace_ArmR]}
       {setName=ParentSpace_Arms, names [ParentSpace_ArmL, ParentSpace_ArmR]}
       {setName=ParentSpace_Legs, names [ParentSpace_LegL, ParentSpace_LegR]}
       {setName=WorldSpace_Legs, names [WorldSpace_LegL, WorldSpace_LegR]}
       {setName=WorldSpace_Hands, names [hand_l,hand_r]}
       {setName=WorldSpace_Feet, names [foot_l,foot_r]}
       {setName=WorldSpace_LowerSpine, names [spine_04,spine_03,spine_02,spine_01,pelvis]}
      ]
      modifierSetUpdates [{setName=hands, names [hand_l,hand_r,weapon_r]}, {setName=feet, names [foot_l,foot_r]}, {setName=Arms, names [ArmL,ArmR]}]
     }
     alphaBoolBlend {blendOption=Linear}
     alphaScaleBiasClamp {bias=1, scale=-1}
     characterSetupData {
      defaultParentSpaceControlData {bEnabled=false}
      defaultWorldSpaceControlData {bEnabled=false}
      limbSetupData [
       {
        bCreateBodyModifiers=true
        bCreateParentSpaceControls=true
        bCreateWorldSpaceControls=true
        bIncludeParentBone=false
        limbName=Head
        startBone=neck_01
       }
       {
        bCreateBodyModifiers=true
        bCreateParentSpaceControls=true
        bCreateWorldSpaceControls=true
        bIncludeParentBone=false
        limbName=ArmL
        startBone=clavicle_l
       }
       {
        bCreateBodyModifiers=true
        bCreateParentSpaceControls=true
        bCreateWorldSpaceControls=true
        bIncludeParentBone=false
        limbName=ArmR
        startBone=clavicle_r
       }
       {
        bCreateBodyModifiers=true
        bCreateParentSpaceControls=true
        bCreateWorldSpaceControls=true
        bIncludeParentBone=false
        limbName=LegL
        startBone=thigh_l
       }
       {
        bCreateBodyModifiers=true
        bCreateParentSpaceControls=true
        bCreateWorldSpaceControls=true
        bIncludeParentBone=false
        limbName=LegR
        startBone=thigh_r
       }
       {
        bCreateBodyModifiers=true
        bCreateParentSpaceControls=true
        bCreateWorldSpaceControls=true
        bIncludeParentBone=true
        limbName=Spine
        startBone=spine_01
       }
      ]
     }
     overrideWorldGravity {z=-980}
     simSpaceSettings {worldAlpha=1, externalLinearDragV=<1 1 1>}
     updateFunction {functionName=On Update_DBNO}
    }
    updateFunction {bSelfContext=true, memberGuid=#3, memberName=On Update_DBNO}
   }
   Save cached pose 'DBNOActionsPBA'(Pose=Component To Local(ComponentPose=@AnimGraphNode_RigidBodyWithControl_0)) {cacheName=DBNOActionsPBA}
   AnimGraphNode_RigidBodyWithControl_1 = RigidBodyWithControl!(
    ComponentPose=Local To Component(LocalPose=Use cached pose 'startCache'())
    Alpha:float=1
    ExternalForce="0,0,0"
    ForcesAndTorques=(ForcesAndTorques=())
   ) -> (Pose) {
    binding=&self:AnimGraph.AnimGraphNode_RigidBodyWithControl_1.Binding
    node {
     bEnableControls=true
     bOverrideWorldGravity=true
     evaluationResetTime=0.2
     overridePhysicsAsset=/LocomotionInitiative/Animation/PBA_Physics_Assets/Fortnite_M_Avg_Player_PBA_Hover
     physicsAssetAuthoredSkeletalMesh=&SK_M_MALE_Base_Skeleton
     physicsAssetConditioningConstraintOrientation=AuthoredSkeleton
     physicsAssetConditioningConstraintPosition=AuthoredSkeleton
     additionalSets {
      controlSetUpdates [
       {setName=ParentSpace_Legs, names [ParentSpace_LegL, ParentSpace_LegR]}
       {setName=WorldSpace_Legs, names [WorldSpace_LegL, WorldSpace_LegR]}
       {setName=WorldSpace_Feet, names [foot_l,foot_r]}
       {setName=WorldSpace_LowerSpine, names [spine_04,spine_03,spine_02,spine_01,pelvis]}
      ]
      modifierSetUpdates [{setName=Legs, names [LegL,LegR]}]
     }
     alphaBoolBlend {blendOption=Linear}
     characterSetupData {
      defaultParentSpaceControlData {bEnabled=false}
      defaultWorldSpaceControlData {bEnabled=false}
      limbSetupData [
       {
        bCreateBodyModifiers=true
        bCreateParentSpaceControls=true
        bCreateWorldSpaceControls=true
        bIncludeParentBone=false
        limbName=Head
        startBone=neck_01
       }
       {
        bCreateBodyModifiers=true
        bCreateParentSpaceControls=true
        bCreateWorldSpaceControls=true
        bIncludeParentBone=false
        limbName=LegL
        startBone=thigh_l
       }
       {
        bCreateBodyModifiers=true
        bCreateParentSpaceControls=true
        bCreateWorldSpaceControls=true
        bIncludeParentBone=false
        limbName=LegR
        startBone=thigh_r
       }
       {
        bCreateBodyModifiers=true
        bCreateParentSpaceControls=true
        bCreateWorldSpaceControls=true
        bIncludeParentBone=true
        limbName=Spine
        startBone=spine_01
       }
      ]
     }
     updateFunction {functionName=OnUpdate_HoverPack}
    }
    updateFunction {bSelfContext=true, memberGuid=#4, memberName=OnUpdate_HoverPack}
   }
   Save cached pose 'HoverPBA'(Pose=Component To Local(ComponentPose=@AnimGraphNode_RigidBodyWithControl_1)) {cacheName=HoverPBA}
   AnimGraphNode_RigidBodyWithControl_3 = RigidBodyWithControl(
    ComponentPose=Local To Component(LocalPose=Use cached pose 'startCache'())
    Alpha=1
    ExternalForce="0,0,0"
    ForcesAndTorques=(ForcesAndTorques=())
   ) {
    binding=&self:AnimGraph.AnimGraphNode_RigidBodyWithControl_3.Binding
    node {
     bEnableControls=true
     bOverrideWorldGravity=true
     evaluationResetTime=0.2
     overridePhysicsAsset=&Fortnite_M_Avg_Player_PBA_Hover
     physicsAssetAuthoredSkeletalMesh=&SK_M_MALE_Base_Skeleton
     physicsAssetConditioningConstraintOrientation=AuthoredSkeleton
     physicsAssetConditioningConstraintPosition=AuthoredSkeleton
     additionalSets {
      controlSetUpdates [
       {setName=ParentSpace_Legs, names [ParentSpace_LegL, ParentSpace_LegR]}
       {setName=WorldSpace_Legs, names [WorldSpace_LegL, WorldSpace_LegR]}
       {setName=WorldSpace_Feet, names [foot_l,foot_r]}
       {setName=WorldSpace_LowerSpine, names [spine_04,spine_03,spine_02,spine_01,pelvis]}
      ]
      modifierSetUpdates [{setName=Legs, names [LegL,LegR]}]
     }
     alphaBoolBlend {blendOption=Linear}
     characterSetupData {
      defaultParentSpaceControlData {bEnabled=false}
      defaultWorldSpaceControlData {bEnabled=false}
      limbSetupData [
       {
        bCreateBodyModifiers=true
        bCreateParentSpaceControls=true
        bCreateWorldSpaceControls=true
        bIncludeParentBone=false
        limbName=Head
        startBone=neck_01
       }
       {
        bCreateBodyModifiers=true
        bCreateParentSpaceControls=true
        bCreateWorldSpaceControls=true
        bIncludeParentBone=false
        limbName=LegL
        startBone=thigh_l
       }
       {
        bCreateBodyModifiers=true
        bCreateParentSpaceControls=true
        bCreateWorldSpaceControls=true
        bIncludeParentBone=false
        limbName=LegR
        startBone=thigh_r
       }
       {
        bCreateBodyModifiers=true
        bCreateParentSpaceControls=true
        bCreateWorldSpaceControls=true
        bIncludeParentBone=true
        limbName=Spine
        startBone=spine_01
       }
      ]
     }
     updateFunction {functionName=OnUpdate_PuppetBoss}
    }
    updateFunction {bSelfContext=true, memberGuid=#5, memberName=OnUpdate_PuppetBoss}
   }
   Save cached pose 'PuppetPBA'(Pose=Component To Local(ComponentPose=@AnimGraphNode_RigidBodyWithControl_3)) {cacheName=PuppetPBA}
   AnimGraphNode_RigidBodyWithControl_5 = RigidBodyWithControl(
    ComponentPose=Local To Component(LocalPose=Use cached pose 'startCache'())
    Alpha=1
    ExternalForce="0,0,0"
    ForcesAndTorques=(ForcesAndTorques=())
   ) {
    binding=&self:AnimGraph.AnimGraphNode_RigidBodyWithControl_5.Binding
    node {
     bEnableControls=true
     bOverrideWorldGravity=true
     evaluationResetTime=0.2
     overridePhysicsAsset=&Fortnite_M_Avg_Player_PBA_Wings
     physicsAssetAuthoredSkeletalMesh=&SK_M_MALE_Base_Skeleton
     physicsAssetConditioningConstraintOrientation=AuthoredSkeleton
     physicsAssetConditioningConstraintPosition=AuthoredSkeleton
     additionalSets {
      controlSetUpdates [
       {setName=ParentSpace_Legs, names [ParentSpace_LegL, ParentSpace_LegR]}
       {setName=WorldSpace_Legs, names [WorldSpace_LegL, WorldSpace_LegR]}
       {setName=WorldSpace_Feet, names [foot_l,foot_r]}
       {setName=WorldSpace_LowerSpine, names [spine_04,spine_03,spine_02,spine_01,pelvis]}
      ]
      modifierSetUpdates [{setName=Legs, names [LegL,LegR]}]
     }
     alphaBoolBlend {blendOption=Linear}
     characterSetupData {
      defaultParentSpaceControlData {bEnabled=false}
      defaultWorldSpaceControlData {bEnabled=false}
      limbSetupData [
       {
        bCreateBodyModifiers=true
        bCreateParentSpaceControls=true
        bCreateWorldSpaceControls=true
        bIncludeParentBone=false
        limbName=LegL
        startBone=thigh_l
       }
       {
        bCreateBodyModifiers=true
        bCreateParentSpaceControls=true
        bCreateWorldSpaceControls=true
        bIncludeParentBone=false
        limbName=LegR
        startBone=thigh_r
       }
       {
        bCreateBodyModifiers=true
        bCreateParentSpaceControls=true
        bCreateWorldSpaceControls=true
        bIncludeParentBone=true
        limbName=Spine
        startBone=spine_01
       }
      ]
     }
     updateFunction {functionName=OnUpdate_IcicleLarva}
    }
    updateFunction {bSelfContext=true, memberGuid=#6, memberName=OnUpdate_IcicleLarva}
   }
   Save cached pose 'IcicleLarva'(Pose=Component To Local(ComponentPose=@AnimGraphNode_RigidBodyWithControl_5)) {cacheName=IcicleLarva}
  ]}
 EventGraph {nodes [
   Event BlueprintUpdateAnimation()
   > n1 = Sequence()
   n2 = Branch(execute=@n1.then_1, Condition=AND Boolean(A=Get bIsDBNO(), B=Get DBNO_CurveTable_Value() -> (DBNO_CurveTable_Value:bool)))
   > Set PBA_DBNOActive(PBA_DBNOActive:bool=true) -> (Output_Get:bool)
   Set PBA_DBNOActive(execute=@n2.else, PBA_DBNOActive=false)
   Set IcicleLarva_MovementState(
    execute=@n1.then_0
    IcicleLarva_MovementState:bool=HasMatchingGameplayTag!(
     self:Gameplay Tag Asset Interface Interface=Get As Fort Player Pawn() -> (As Fort Player Pawn:Fort Player Pawn Object Reference)
     TagToCheck:Gameplay Tag Structure=(TagName="IcicleLarva.MovementState")
    ) -> (ReturnValue:bool)
   ) -> (Output_Get:bool)
   Event BlueprintBeginPlay() {is=title K2Node_Event}
   > n3 = Cast To FortPlayerPawn(Object=GetOwningActor())
   > Set As Fort Player Pawn(As Fort Player Pawn:Fort Player Pawn Object Reference=@n3) -> (Output_Get:Fort Player Pawn Object Reference)
   > n4 = Sequence()
   n5 = GetOwnerFortPlayerAnimInstance!(self:Custom Character Part Anim Instance Object Reference) -> (ReturnValue:Fort Player Anim Instance Object Reference)
   > AddCharacterDynamicsControlLogicLayer!(
    execute
    InAnimInstance:Fort Anim Instance Object Reference=@n5
    LayerName:Name=SkydivingPBA
    LayerStateLogic:Fort Character Dynamics State Logic Object Reference=/CharacterDynamicsControl/CustomGameplayItems/SkydivingPBA_StateLogic
   ) -> (then)
   > AddCharacterDynamicsControlLogicLayer(InAnimInstance=@n5, LayerName=FlyingWings, LayerStateLogic=/CharacterDynamicsControl/CustomGameplayItems/SunroseWings_StateLogic)
   > AddCharacterDynamicsControlLogicLayer(InAnimInstance=@n5, LayerName=HoverPack, LayerStateLogic=/CharacterDynamicsControl/CustomGameplayItems/HoverPack_StateLogic)
   AddCharacterDynamicsControlLogicLayer(
    execute=@n4.then_1
    InAnimInstance=@n5
    LayerName=PuppetBoss
    LayerStateLogic=/CharacterDynamicsControl/CustomGameplayItems/PuppetBoss_StateLogic
   )
   > AddCharacterDynamicsControlLogicLayer(InAnimInstance=@n5, LayerName=IcicleLarva, LayerStateLogic=/CharacterDynamicsControl/CustomGameplayItems/IcicleLarva_StateLogic)
   > AddCharacterDynamicsControlLogicLayer(InAnimInstance=@n5, LayerName=DBNO, LayerStateLogic=/CharacterDynamicsControl/CustomGameplayItems/DBNO_StateLogic)
   Event BlueprintInitializeAnimation()
   > Set DBNO_CurveTable_Value(DBNO_CurveTable_Value:bool=GetValueAsBool!(Input:"Scalable Float Structure (by ref)"=Get DBNO_PBA_Enabled() -> (DBNO_PBA_Enabled:Scalable Float Structure), Level:float=0) -> (ReturnValue:bool)) -> (Output_Get:bool)
  ]}
 On Update_DBNO {nodes [
   n1 = On Update_DBNO!() -> (then, Context:"Anim Update Context Structure (by ref)", Node:"Anim Node Reference Structure (by ref)") {is=title K2Node_FunctionEntry}
   > n2 = ConvertToRigidBodyWithControl!(execute, Node:"Anim Node Reference Structure (by ref)"=@n1.Node) -> (Succeeded:Exec, Failed:Exec, ReturnValue:Rigid Body With Control Reference Structure)
   Return Node(execute=[@n2.Failed, @n3.then])
   > n3 = UpdateRigidBodyWithControlNodeParametersForLayer!(
    execute
    UpdateContext:"Anim Update Context Structure (by ref)"=@n1
    RigidBodyWithControl:"Rigid Body With Control Reference Structure (by ref)"=@n2
    StateLogicLayerName:Name=DBNO
    Parameters:Fort Character Dynamics Parameters Object Reference=/CharacterDynamicsControl/CustomGameplayItems/DBNO_Parameters
   ) -> (then, ReturnValue:Rigid Body With Control Reference Structure)
  ]}
 On Update_Skydiving {nodes [
   n1 = On Update_Skydiving!() -> (then, Context:"Anim Update Context Structure (by ref)", Node:"Anim Node Reference Structure (by ref)") {is=title K2Node_FunctionEntry}
   > n2 = ConvertToRigidBodyWithControl(Node=@n1.Node)
   Return Node(execute=[@n2.Failed, @n3.then])
   > n3 = UpdateRigidBodyWithControlNodeParametersForLayer(
    UpdateContext=@n1
    RigidBodyWithControl=@n2
    StateLogicLayerName=SkydivingPBA
    Parameters=/CharacterDynamicsControl/CustomGameplayItems/SkyDive_Parameters
   )
  ]}
 OnUpdate_HoverPack {nodes [
   n1 = OnUpdate_HoverPack!() -> (then, Context:"Anim Update Context Structure (by ref)", Node:"Anim Node Reference Structure (by ref)") {is=title K2Node_FunctionEntry}
   > n2 = ConvertToRigidBodyWithControl(Node=@n1.Node)
   Return Node(execute=[@n2.Failed, @n3.then])
   > n3 = UpdateRigidBodyWithControlNodeParametersForLayer(
    UpdateContext=@n1
    RigidBodyWithControl=@n2
    StateLogicLayerName=HoverPack
    Parameters=/CharacterDynamicsControl/CustomGameplayItems/CustomGameplay_Parameters
   )
  ]}
 OnUpdate_IcicleLarva {nodes [
   n1 = OnUpdate_IcicleLarva!() -> (then, Context:"Anim Update Context Structure (by ref)", Node:"Anim Node Reference Structure (by ref)") {is=title K2Node_FunctionEntry}
   > n2 = ConvertToRigidBodyWithControl(Node=@n1.Node)
   Return Node(execute=[@n2.Failed, @n3.then])
   > n3 = UpdateRigidBodyWithControlNodeParametersForLayer(
    UpdateContext=@n1
    RigidBodyWithControl=@n2
    StateLogicLayerName=IcicleLarva
    Parameters=&CustomGameplay_Parameters
   )
  ]}
 OnUpdate_PuppetBoss {nodes [
   n1 = OnUpdate_PuppetBoss!() -> (then, Context:"Anim Update Context Structure (by ref)", Node:"Anim Node Reference Structure (by ref)") {is=title K2Node_FunctionEntry}
   > n2 = ConvertToRigidBodyWithControl(Node=@n1.Node)
   Return Node(execute=[@n2.Failed, @n3.then])
   > n3 = UpdateRigidBodyWithControlNodeParametersForLayer(
    UpdateContext=@n1
    RigidBodyWithControl=@n2
    StateLogicLayerName=PuppetBoss
    Parameters=&CustomGameplay_Parameters
   )
  ]}
 OnUpdate_Wings {nodes [
   n1 = OnUpdate_Wings!() -> (then, Context:"Anim Update Context Structure (by ref)", Node:"Anim Node Reference Structure (by ref)") {is=title K2Node_FunctionEntry}
   > n2 = ConvertToRigidBodyWithControl(Node=@n1.Node)
   Return Node(execute=[@n2.Failed, @n3.then])
   > n3 = UpdateRigidBodyWithControlNodeParametersForLayer(
    UpdateContext=@n1
    RigidBodyWithControl=@n2
    StateLogicLayerName=FlyingWings
    Parameters=&CustomGameplay_Parameters
   )
  ]}
 TrackFastAnimationMovements {nodes [
   TrackFastAnimationMovements() {is=title K2Node_FunctionEntry}
   > n1 = Sequence()
   > Branch(Condition="Not Equal (Name)"(
     A=Get CurrentControlStateName() -> (CurrentControlStateName:Name=None)
     B=Get PreviousControlStateName() -> (PreviousControlStateName:Name=None)
    ) -> (ReturnValue=~))
   > Set PreviousControlStateName(PreviousControlStateName:Name=Get CurrentControlStateName() -> (CurrentControlStateName=None)) -> (Output_Get:Name=None)
   > n2 = Switch on Name!(execute:Exec, Selection:Name=Get CurrentControlStateName() -> (CurrentControlStateName=None)) -> (Default:Exec, Skydive_Parachuting:Exec) {is=title K2Node_SwitchName, pinNames [Skydive_Parachuting]}
   > n3 = Set ControlParamTransitionTime(ControlParamTransitionTime:double=0) -> (Output_Get:double)
   Set ControlParamTransitionRemainingTime(
    execute=[@n4.then, @n3.then]
    ControlParamTransitionRemainingTime:double=Get ControlParamTransitionTime() -> (ControlParamTransitionTime:double)
   ) -> (Output_Get:double)
   n4 = Set ControlParamTransitionTime(execute=@n2.Skydive_Parachuting, ControlParamTransitionTime=2)
   Branch(execute=@n1.then_1, Condition="float > float"(A=Get ControlParamTransitionRemainingTime() -> (ControlParamTransitionRemainingTime:double)))
   > n5 = Set ControlParamTransitionRemainingTime(ControlParamTransitionRemainingTime=float - float(A=Get ControlParamTransitionRemainingTime(), B=GetDeltaSeconds!(self:Anim Instance Object Reference) -> (ReturnValue:float)))
   > Branch(Condition="float > float"(A=@n5))
   > Set Control And Modifier Parameters Final(Control And Modifier Parameters Final:Physics Control Control And Modifier Parameters Structure=AddControlParameters!(
     InParameters:"Physics Control Control And Modifier Parameters Structure (by ref)"=Get Control And Modifier Parameters Final() -> (Control And Modifier Parameters Final:Physics Control Control And Modifier Parameters Structure)
     Name:Name=WorldSpace_Spine
     ControlData:"Physics Control Sparse Data Structure (by ref)"=Make Physics Control Sparse Data!(
      LinearStrength:float=1000
      LinearDampingRatio:float=1
      LinearExtraDamping:float=0
      MaxForce:float=0
      AngularStrength:float=1000
      AngularDampingRatio:float=1
      AngularExtraDamping:float=0
      MaxTorque:float=0
      LinearTargetVelocityMultiplier:float=1
      AngularTargetVelocityMultiplier:float=1
      CustomControlPoint:Vector="0,0,0"
      bEnabled:bool=True
      bUseCustomControlPoint:bool=False
      bUseSkeletalAnimation:bool=True
      bDisableCollision:bool=False
      bOnlyControlChildObject:bool=False
      bUseAccelerationDriveMode:bool=True
     ) -> (PhysicsControlSparseData:Physics Control Sparse Data Structure) {is=title K2Node_MakeStruct}
    ) -> (OutParameters:Physics Control Control And Modifier Parameters Structure)) -> (Output_Get:Physics Control Control And Modifier Parameters Structure)
   n6 = float / float(A=Get ControlParamTransitionRemainingTime(), B=Get ControlParamTransitionTime())
   float * float(A="Sin (Degrees)"!(A:double=float * float(A=@n6, B=180)) -> (ReturnValue:double), B=100) {is=title K2Node_PromotableOperator}
  ]}
]
variables [
 As Fort Player Pawn {type=FortPlayerPawn}
 Control And Modifier Parameters Final {type=PhysicsControlControlAndModifierParameters, default {controlMultiplierParameters [], controlParameters [], modifierParameters []}}
 ControlParamTransitionRemainingTime {type=number}
 ControlParamTransitionTime {type=number}
 CurrentControlStateName {type=string, default=None}
 DBNO_CurveTable_Value {type=boolean}
 DBNO_PBA_Enabled {
  type=ScalableFloat
  default {registryType=None, value=1, curve {curveTable=/LocomotionInitiative/DataTables/PBACurveData, rowName=TrendConcentrate.DBNO.PBA.Enabled}}
 }
 IcicleLarva_MovementState {type=boolean}
 PBA_DBNOActive {type=boolean}
 PreviousControlStateName {type=string, default=None}
]
